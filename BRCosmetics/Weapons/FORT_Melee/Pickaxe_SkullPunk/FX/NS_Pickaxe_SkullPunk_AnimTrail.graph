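# Niagara swing trail effect for the SkullPunk pickaxe, built from four ribbon emitters with slightly staggered loop delays.
# Blade, Block and Equalizer use the shared AnimTrail modules to spawn short-lived ribbons from bone or socket indices, fading color and alpha by curves with a camera offset.
# Trail_Swing instead rate-spawns a tapering ribbon at a socket on the user skeletal mesh, shifting color over its life.
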NiagaraSystem /BRCosmetics/Weapons/FORT_Melee/Pickaxe_SkullPunk/FX/NS_Pickaxe_SkullPunk_AnimTrail
dependencies {
 usedDataInterfaces [&NiagaraDataInterfaceCurve, &NiagaraDataInterfaceColorCurve]
 usedDynamicInputs []
 usedModules [&SystemState]
}
emitters {
 Blade {
  emitterUpdateScript [
   &EmitterState(Life Cycle Mode=::NewEnumerator1, Loop Behavior=::NewEnumerator1, Loop Delay=0.05, Loop Duration=0.5, UseLoopDelay=true)
   &FN_SpawnAnimTrail_v2(AnimTrailWidthMode=::FromSecond)
  ]
  particleSpawnScript [
   /Game/Effects/Niagara/Modules/AnimTrail/FN_InitializeAnimTrail/FN_InitializeAnimTrailBoneIndecies(Second Bone or Socket Index=1)
   /Game/Effects/Niagara/Modules/AnimTrail/FN_InitializeAnimTrail/FN_InitializeAnimTrail(AnimTrailWidth=0.95, Color=<rgba 1 1 1 3>, Lifetime=0.27)
  ]
  particleUpdateScript [
   &FN_UpdateAnimTrail_v2()
   &SpriteFacingAndAlignment(Sprite Facing=$Particles.RibbonFacing, WriteAlignment=false, WriteFacing=true)
   &ScaleRibbonWidth(Ribbon Width Scale=&FloatFromCurve(FloatCurve=DataInterface() {propertyValues {Curve {keys [
         {arriveTangent=-0.9853, interpMode=RCIM_Cubic, leaveTangent=-0.9853, tangentMode=RCTM_User, time=0.4, value=1}
         {arriveTangent=-0.4737, interpMode=RCIM_Cubic, leaveTangent=-0.4737, tangentMode=RCTM_User, time=1, value=0.4}
        ]}}}))
   &ScaleColor(
    Scale Alpha=&FloatFromCurve(FloatCurve=DataInterface() {propertyValues {Curve {keys [
         {interpMode=RCIM_Cubic, leaveTangent=0.0089, tangentMode=RCTM_Break, time=0, value=1}
         {arriveTangent=-0.5839, interpMode=RCIM_Cubic, leaveTangent=-0.5839, tangentMode=RCTM_User, time=0.3, value=0.15}
         {interpMode=RCIM_Cubic, tangentMode=RCTM_Break, time=1, value=0}
        ]}}})
    Scale RGB=&VectorFromFloat(Value=&FloatFromCurve(FloatCurve=DataInterface() {propertyValues {Curve {keys [
          <at 0 1>
          {interpMode=RCIM_Cubic, tangentMode=RCTM_User, time=0.2, value=1}
          {arriveTangent=-1.3221, interpMode=RCIM_Cubic, leaveTangent=-1.3221, tangentMode=RCTM_User, time=0.8, value=0.375}
          {arriveTangent=-1.771, interpMode=RCIM_Cubic, leaveTangent=-1.771, tangentMode=RCTM_User, time=1, value=0}
         ]}}}))
    ScaleA=true
    ScaleRGB=true
   )
   &CameraOffset(Camera Offset Amount=2)
  ]
  rendererClasses [&NiagaraRibbonRendererProperties]
  renderers [{rendererClass=&NiagaraRibbonRendererProperties}]
 }
 Block {
  emitterUpdateScript [
   &EmitterState(Life Cycle Mode=::NewEnumerator1, Loop Behavior=::NewEnumerator1, Loop Delay=0.06, Loop Duration=0.5, UseLoopDelay=true)
   &FN_SpawnAnimTrail_v2(AnimTrailWidthMode=::FromSecond)
  ]
  particleSpawnScript [
   &FN_InitializeAnimTrailBoneIndecies(Second Bone or Socket Index=1)
   &FN_InitializeAnimTrail(AnimTrailWidth=0.8, Color=<rgba 0.5 0.25 0.4583 10>, Lifetime=0.28)
  ]
  particleUpdateScript [
   &FN_UpdateAnimTrail_v2()
   &SpriteFacingAndAlignment(Sprite Facing=$Particles.RibbonFacing, WriteAlignment=false, WriteFacing=true)
   &ScaleColor(
    Scale Alpha=&FloatFromCurve(FloatCurve=DataInterface() {propertyValues {Curve {keys [
         {interpMode=RCIM_Cubic, leaveTangent=0.0089, tangentMode=RCTM_Break, time=0, value=1}
         {arriveTangent=-0.5839, interpMode=RCIM_Cubic, leaveTangent=-0.5839, tangentMode=RCTM_User, time=0.3, value=0.25}
         {interpMode=RCIM_Cubic, tangentMode=RCTM_Break, time=1, value=0}
        ]}}})
    Scale RGB=&VectorFromFloat(Value=&FloatFromCurve(FloatCurve=DataInterface() {propertyValues {Curve {keys [
          <at 0 1>
          {interpMode=RCIM_Cubic, tangentMode=RCTM_User, time=0.4, value=1}
          {arriveTangent=-1.3221, interpMode=RCIM_Cubic, leaveTangent=-1.3221, tangentMode=RCTM_User, time=0.8, value=0.375}
          {arriveTangent=-1.771, interpMode=RCIM_Cubic, leaveTangent=-1.771, tangentMode=RCTM_User, time=1, value=0}
         ]}}}))
    ScaleA=true
    ScaleRGB=true
   )
   &CameraOffset(Camera Offset Amount=5)
  ]
  rendererClasses [&NiagaraRibbonRendererProperties]
  renderers [{rendererClass=&NiagaraRibbonRendererProperties}]
 }
 Equalizer {
  emitterUpdateScript [
   &EmitterState(Life Cycle Mode=::NewEnumerator1, Loop Behavior=::NewEnumerator1, Loop Delay=0.05, Loop Duration=0.5, UseLoopDelay=true)
   &FN_SpawnAnimTrail_v2(AnimTrailWidthMode=::FromSecond)
  ]
  particleSpawnScript [
   &FN_InitializeAnimTrailBoneIndecies(Second Bone or Socket Index=1)
   &FN_InitializeAnimTrail(AnimTrailWidth=0.8, Color=<rgba 4 4 4 10>, Lifetime=0.09)
  ]
  particleUpdateScript [
   &FN_UpdateAnimTrail_v2()
   &SpriteFacingAndAlignment(Sprite Facing=$Particles.RibbonFacing, WriteAlignment=false, WriteFacing=true)
   &ScaleColor(
    Scale Alpha=&FloatFromCurve(FloatCurve=DataInterface() {propertyValues {Curve {keys [
         {interpMode=RCIM_Cubic, leaveTangent=0.0089, tangentMode=RCTM_Break, time=0, value=1}
         {arriveTangent=-0.5839, interpMode=RCIM_Cubic, leaveTangent=-0.5839, tangentMode=RCTM_User, time=0.3, value=0.15}
         {interpMode=RCIM_Cubic, tangentMode=RCTM_Break, time=1, value=0}
        ]}}})
    Scale RGB=&VectorFromFloat(Value=&FloatFromCurve(FloatCurve=DataInterface() {propertyValues {Curve {keys [
          <at 0 1>
          {interpMode=RCIM_Cubic, tangentMode=RCTM_User, time=0.4, value=1}
          {arriveTangent=-1.3221, interpMode=RCIM_Cubic, leaveTangent=-1.3221, tangentMode=RCTM_User, time=0.8, value=0.375}
          {arriveTangent=-1.771, interpMode=RCIM_Cubic, leaveTangent=-1.771, tangentMode=RCTM_User, time=1, value=0}
         ]}}}))
    ScaleA=true
    ScaleRGB=true
   )
   &CameraOffset(Camera Offset Amount=5)
  ]
  rendererClasses [&NiagaraRibbonRendererProperties]
  renderers [{rendererClass=&NiagaraRibbonRendererProperties}]
 }
 Trail_Swing {
  emitterUpdateScript [
   &EmitterState(Life Cycle Mode=::NewEnumerator1, Loop Behavior=::NewEnumerator1, Loop Delay=0.03, Loop Duration=1, UseLoopDelay=true)
   &SpawnRate(SpawnRate=45, Use Spawn Probability=false)
  ]
  particleSpawnScript [
   &InitializeRibbon(
    Lifetime=0.25
    Module.Write Color=true
    Module.Write Lifetime=true
    Module.Write Mass=true
    Module.Write Material Random=true
    Module.Write Position=false
    Module.Write RibbonTwist=false
    Module.Write RibbonWidth=true
    Ribbon Width=20
   )
   &SkeletalMeshLocation(
    Bone / Socket Index=1
    Inherit Velocity=false
    MaskSpawnGroup=false
    Mesh Sampling Type=::NewEnumerator5
    Sampled Position Offset=<0 0 -3>
    Sampled Position Offset Coordinate Space:ENiagaraMeshTransforms=::NewEnumerator3
    Skeletal Mesh=$User.SkeletalMesh
    Socket Sampling Mode=::NewEnumerator5
   )
  ]
  particleUpdateScript [
   &ParticleState()
   &SolveForcesAndVelocity(Clamp Velocity=false, Limit Acceleration=false)
   &ScaleRibbonWidth(Ribbon Width Scale=&FloatFromCurve(FloatCurve=DataInterface() {propertyValues {Curve {keys [<at 0.25 1 RCIM_Cubic>, {arriveTangent=-3.3432, interpMode=RCIM_Cubic, leaveTangent=-3.3432, tangentMode=RCTM_User, time=1, value=0}]}}})) {module=ScaleRibbonWidth001}
   &Color(Color=&ColorFromCurve(ColorCurve=DataInterface() {propertyValues {
       AlphaCurve {keys [<at 0 1 RCIM_Cubic>,<at 1 0 RCIM_Cubic>]}
       BlueCurve {keys [<at 0 1.6667 RCIM_Cubic>,<at 0.4822 0.375>,<at 0.8839 0.5 RCIM_Cubic>]}
       GreenCurve {keys [<at 0 2 RCIM_Cubic>,<at 0.4822 0.0188>,<at 0.8839 0 RCIM_Cubic>]}
       RedCurve {keys [<at 0 0 RCIM_Cubic>,<at 0.4822 0.0781>,<at 0.8839 1 RCIM_Cubic>]}
      }}))
  ]
  rendererClasses [&NiagaraRibbonRendererProperties]
  renderers [{rendererClass=&NiagaraRibbonRendererProperties}]
 }
}
summary {
 systemName=NS_Pickaxe_SkullPunk_AnimTrail
 emitters [
  {emitterName=Blade, rendererClasses [&NiagaraRibbonRendererProperties]}
  {emitterName=Equalizer, rendererClasses [&NiagaraRibbonRendererProperties]}
  {emitterName=Block, rendererClasses [&NiagaraRibbonRendererProperties]}
  {emitterName=Trail_Swing, rendererClasses [&NiagaraRibbonRendererProperties]}
 ]
 userVariables [
  User.RibbonIDOffset {type=NiagaraInt32}
  User.SkeletalMesh {
   type=NiagaraDataInterfaceSkeletalMesh
   defaultValue {
    struct=NiagaraExt_VariableValue_DataInterface
    value {dataInterface=&self:NiagaraDataInterfaceSkeletalMesh_1}
   }
  }
 ]
}
