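# Niagara idle effect for the SkullPunk pickaxe, flickering short-lived glitch streaks around the weapon.
# Three near-identical looping sprite emitters (Top, Center, Under) spawn thin bars within separate offset bands, each tinted either cyan or red at random.
# Bars use fixed facing and alignment, halve in height over life, and feed randomized dynamic material parameters to two sprite renderers.
# The shown modules reference neither declared user variable (Color A, Offset Back).
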NiagaraSystem /BRCosmetics/Weapons/FORT_Melee/Pickaxe_SkullPunk/FX/NS_Pickaxe_SkullPunk_Idle
dependencies {
 usedDataInterfaces [&NiagaraDataInterfaceVector2DCurve]
 usedDynamicInputs []
 usedModules [&SystemState]
}
emitters {
 Glitch_Center {
  emitterUpdateScript [
   &EmitterState(Life Cycle Mode=::NewEnumerator1, Loop Duration=1, UseLoopDelay=false)
   &SpawnRate(Use Spawn Probability=false)
  ]
  particleSpawnScript [
   &FN_SetSpawnGroup(Normalize Spawn Group=true, Particles Per Group=1, Pattern=::NewEnumerator0)
   &InitializeParticle(
    Color=&LerpLinearColors(
     StartColor=<rgba 0 1.5 1 2>
     EndColor=<rgba 0.8 0.04 0.1033 1>
     LerpFactor=&UniformAOrBFloat(A=0, Distribution Weight=0.4)
    )
    Lifetime Max=0.5
    Lifetime Min=0.4
    Lifetime Mode=::NewEnumerator1
    Sprite Size=&Multiply_Vector2D_ByFloat(Float=0.5, Vector2D=&RandomRangeVector2D(Minimum=<20 2.5>, Maximum=<23 3>))
    Sprite Size Mode=::NewEnumerator4
    UsePositionOffset=true
    Write Material Random=true
   )
   &SystemLocation(MaskSpawnGroup=false, Offset=&RandomRangeVector(Minimum=<15 0 45>, Maximum=<3 0 10>))
   &DynamicMaterialParameters(
    Index 0 Param 1=&RandomRangeFloat(Minimum=0.05, Maximum=0.05)
    Index 0 Param 2=&RandomRangeFloat(Minimum=3, Maximum=1.5)
    Index 0 Param 3=&RandomRangeFloat(Minimum=1, Maximum=0.5)
    Index 0 Param 4=1
    Index 1 Param 1=1
    Module.Param0WriteEnabled=true
    Module.Param0WriteEnabled1=true
    Module.Param1WriteEnabled=true
    Module.Param1WriteEnabled1=true
    Module.Param2WriteEnabled=true
    Module.Param2WriteEnabled1=true
    Module.Param3WriteEnabled=true
    Module.Param3WriteEnabled1=true
    Write Parameter Index 1=true
   )
  ]
  particleUpdateScript [
   &ParticleState()
   &SolveForcesAndVelocity(Clamp Velocity=false, Limit Acceleration=false)
   &ScaleSpriteSize(Scale Factor=&Vector2DFromCurve(Vector2Curve=DataInterface() {propertyValues {XCurve {keys [<at 0 1>,<at 1 1>]}, YCurve {keys [<at 0 1>,<at 1 0.5>]}}}, ScaleCurve=<1 1>))
   &SpriteFacingAndAlignment(Sprite Alignment=<0 0 1>, Sprite Facing=<0 -1 0>, WriteAlignment=true, WriteFacing=true)
   &CameraOffset(Camera Offset Amount=10)
  ]
  rendererClasses [&NiagaraSpriteRendererProperties]
  renderers [
   {rendererClass=&NiagaraSpriteRendererProperties}
   {rendererClass=&NiagaraSpriteRendererProperties, rendererIndex=1}
  ]
 }
 Glitch_Top {
  emitterUpdateScript [
   &EmitterState(Life Cycle Mode=::NewEnumerator1, Loop Duration=1, UseLoopDelay=false)
   &SpawnRate(SpawnRate=2, Use Spawn Probability=false)
  ]
  particleSpawnScript [
   &FN_SetSpawnGroup(Normalize Spawn Group=true, Particles Per Group=1, Pattern=::NewEnumerator0)
   &InitializeParticle(
    Color=&LerpLinearColors(
     StartColor=<rgba 0 1.5 1 2>
     EndColor=<rgba 0.8 0.04 0.1033 1>
     LerpFactor=&UniformAOrBFloat(A=0, Distribution Weight=0.4)
    )
    Lifetime Max=0.7
    Lifetime Min=0.3
    Lifetime Mode=::NewEnumerator1
    Sprite Size=&Multiply_Vector2D_ByFloat(Float=0.5, Vector2D=&RandomRangeVector2D(Minimum=<20 2.5>, Maximum=<23 3>))
    Sprite Size Mode=::NewEnumerator4
    UsePositionOffset=true
    Write Material Random=true
   )
   &SystemLocation(MaskSpawnGroup=false, Offset=&RandomRangeVector(Minimum=<45 0 57>, Maximum=<-10 0 45>))
   &DynamicMaterialParameters(
    Index 0 Param 1=&RandomRangeFloat(Minimum=0.05, Maximum=0.05)
    Index 0 Param 2=&RandomRangeFloat(Minimum=3, Maximum=1.5)
    Index 0 Param 3=&RandomRangeFloat(Minimum=1, Maximum=0.5)
    Index 0 Param 4=1
    Index 1 Param 1=1
    Module.Param0WriteEnabled=true
    Module.Param0WriteEnabled1=true
    Module.Param1WriteEnabled=true
    Module.Param1WriteEnabled1=true
    Module.Param2WriteEnabled=true
    Module.Param2WriteEnabled1=true
    Module.Param3WriteEnabled=true
    Module.Param3WriteEnabled1=true
    Write Parameter Index 1=true
   )
  ]
  particleUpdateScript [
   &ParticleState()
   &SolveForcesAndVelocity(Clamp Velocity=false, Limit Acceleration=false)
   &ScaleSpriteSize(Scale Factor=&Vector2DFromCurve(Vector2Curve=DataInterface() {propertyValues {XCurve {keys [<at 0 1>,<at 1 1>]}, YCurve {keys [<at 0 1>,<at 1 0.5>]}}}, ScaleCurve=<1 1>))
   &SpriteFacingAndAlignment(Sprite Alignment=<0 0 1>, Sprite Facing=<0 -1 0>, WriteAlignment=true, WriteFacing=true)
   &CameraOffset(Camera Offset Amount=10)
  ]
  rendererClasses [&NiagaraSpriteRendererProperties]
  renderers [
   {rendererClass=&NiagaraSpriteRendererProperties}
   {rendererClass=&NiagaraSpriteRendererProperties, rendererIndex=1}
  ]
 }
 Glitch_Under {
  emitterUpdateScript [
   &EmitterState(Life Cycle Mode=::NewEnumerator1, Loop Duration=1, UseLoopDelay=false)
   &SpawnRate(SpawnRate=1.5, Use Spawn Probability=false)
  ]
  particleSpawnScript [
   &FN_SetSpawnGroup(Normalize Spawn Group=true, Particles Per Group=1, Pattern=::NewEnumerator0)
   &InitializeParticle(
    Color=&LerpLinearColors(
     StartColor=<rgba 0 1.5 1 2>
     EndColor=<rgba 0.8 0.04 0.1033 1>
     LerpFactor=&UniformAOrBFloat(A=0, Distribution Weight=0.4)
    )
    Lifetime Max=0.5
    Lifetime Min=0.4
    Lifetime Mode=::NewEnumerator1
    Sprite Size=&Multiply_Vector2D_ByFloat(Float=0.5, Vector2D=&RandomRangeVector2D(Minimum=<17 2.2>, Maximum=<21 2.7>))
    Sprite Size Mode=::NewEnumerator4
    UsePositionOffset=true
    Write Material Random=true
   )
   &SystemLocation(MaskSpawnGroup=false, Offset=&RandomRangeVector(Minimum=<10 0 -40>, Maximum=<-3 0 -60>))
   &DynamicMaterialParameters(
    Index 0 Param 1=&RandomRangeFloat(Minimum=0.05, Maximum=0.05)
    Index 0 Param 2=&RandomRangeFloat(Minimum=3, Maximum=1.5)
    Index 0 Param 3=&RandomRangeFloat(Minimum=1, Maximum=0.5)
    Index 0 Param 4=1
    Index 1 Param 1=1
    Module.Param0WriteEnabled=true
    Module.Param0WriteEnabled1=true
    Module.Param1WriteEnabled=true
    Module.Param1WriteEnabled1=true
    Module.Param2WriteEnabled=true
    Module.Param2WriteEnabled1=true
    Module.Param3WriteEnabled=true
    Module.Param3WriteEnabled1=true
    Write Parameter Index 1=true
   )
  ]
  particleUpdateScript [
   &ParticleState()
   &SolveForcesAndVelocity(Clamp Velocity=false, Limit Acceleration=false)
   &ScaleSpriteSize(Scale Factor=&Vector2DFromCurve(Vector2Curve=DataInterface() {propertyValues {XCurve {keys [<at 0 1>,<at 1 1>]}, YCurve {keys [<at 0 1>,<at 1 0.5>]}}}, ScaleCurve=<1 1>))
   &SpriteFacingAndAlignment(Sprite Alignment=<0 0 1>, Sprite Facing=<0 -1 0>, WriteAlignment=true, WriteFacing=true)
   &CameraOffset(Camera Offset Amount=10)
  ]
  rendererClasses [&NiagaraSpriteRendererProperties]
  renderers [
   {rendererClass=&NiagaraSpriteRendererProperties}
   {rendererClass=&NiagaraSpriteRendererProperties, rendererIndex=1}
  ]
 }
}
summary {
 systemName=NS_Pickaxe_SkullPunk_Idle
 emitters [
  {emitterName=Glitch_Top, rendererClasses [&NiagaraSpriteRendererProperties]}
  {emitterName=Glitch_Under, rendererClasses [&NiagaraSpriteRendererProperties]}
  {emitterName=Glitch_Center, rendererClasses [&NiagaraSpriteRendererProperties]}
 ]
 userVariables [
  User.Color A {type=LinearColor, description="", defaultValue {struct=LinearColor, value=<rgba 1 0 0 1>}}
  User.Offset Back {type=NiagaraFloat, defaultValue=0, description=""}
 ]
}
